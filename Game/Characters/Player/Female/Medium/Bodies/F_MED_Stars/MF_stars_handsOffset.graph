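# Material function for the female medium Stars player body that outputs a mask around both hands.
# It takes the X and Z of the pre-skinned position, so the mask stays fixed to the mesh during animation and ignores depth along Y.
# Two sphere masks around left and right hand position parameters share radius and hardness parameters, then are added and saturated into one output.
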MaterialFunction /Game/Characters/Player/Female/Medium/Bodies/F_MED_Stars/MF_stars_handsOffset
graph {nodes [
  n1 = ComponentMask(None=PreSkinnedPosition()) {b=true, r=true}
  n2 = VectorParameter() {group=StarsHands, parameterName=Stars_leftHandPosition, sortPriority=3, defaultValue=<rgba 55 95 0 1>}
  n3 = ScalarParameter() {defaultValue=18, group=StarsHands, parameterName=Stars_HandRadius, sortPriority=1}
  n4 = ScalarParameter() {group=StarsHands, parameterName=Stars_HandHardness, sortPriority=2}
  n5 = SphereMask(A=@n1, B=ComponentMask(None=@n2.RGB) {g=true, r=true}, Radius=@n3, Hardness=@n4) {attenuationRadius=25, hardnessPercent=0}
  n6 = VectorParameter() {group=StarsHands, parameterName=Stars_rightHandPosition, sortPriority=3, defaultValue=<rgba -55 95 0 1>}
  n7 = SphereMask(A=@n1, B=ComponentMask(None=@n6.RGB) {g=true, r=true}, Radius=@n3, Hardness=@n4) {attenuationRadius=25, hardnessPercent=0}
  FunctionOutput(None=Saturate(None=Add(A=@n5, B=@n7)))
 ]}
