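# Material function for the black part of the Stars female player body, switching statically between textured wireframe and inner/outer shell looks.
# Wireframe tints a neon grid mask with a scrolling, hue-cycling three-color gradient, using pre-skinned position as UV1.
# Otherwise, floored UV0 U selects inner or outer surface values and camera offsets, with optional hand adjustment.
# Under ray tracing, outer vertices collapse to the object position and opacity becomes parameter-driven.
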MaterialFunction /Game/Characters/Player/Female/Medium/Bodies/F_MED_Stars/Materials/MF_stars_black
graph {nodes [
  Constant3Vector() {constant=<rgba 0 0.12 1 0>}
  n1 = GetMaterialAttributes(None=FunctionInput() {inputType=FunctionInput_MaterialAttributes}) {attributeGetTypes [#1, #2, #3]}
  n2 = TextureSampleParameter2D() {
   group=Neon
   parameterName=Emissive
   samplerType=SAMPLERTYPE_Masks
   texture=/Game/Characters/Player/Male/Medium/Bodies/M_MED_Neon/Textures/neon_Grid_10x
  }
  n3 = Time()
  n4 = TextureCoordinate() {coordinateIndex=1}
  n5 = Panner(Coordinate=Divide(A=ComponentMask(None=@n4) {g=true}) {constB=140}) {bFractionalPart=true, speedX=0.1}
  n6 = MaterialFunctionCall(
   Hue Shift Percentage=Frac(None=Multiply(A=@n3) {constB=0.2})
   Texture=MaterialFunctionCall(
    A=Constant3Vector() {constant=<rgba 0.1015 0 1 0>}
    B=Constant3Vector() {constant=<rgba 0.855 0.0343 0.7913 0>}
    C=Constant3Vector() {constant=<rgba 0.1022 0 1 1>}
    Alpha=Frac(None=@n5)
   ) {materialFunction=&3ColorBlend}.Result
  ) {materialFunction=&HueShift}
  n7 = VectorParameter() {group=Stars, parameterName=Stars_BaseColor_inner, defaultValue=<rgba 0.009 0.009 0.009 0>}
  n8 = VectorParameter() {group=Stars, parameterName=Stars_MSR_inner, defaultValue=<rgba 0 0.5 0.3 0>}
  n9 = MaterialFunctionCall(Normal=@n1.Normal, Flatness=ScalarParameter() {group=Stars, parameterName=Stars_flattenNormal}) {materialFunction=&FlattenNormal}
  n10 = SetMaterialAttributes!(
   MaterialAttributes=@n1.MaterialAttributes
   EmissiveColor=Multiply(
    A=ScalarParameter() {defaultValue=1, group=Stars, parameterName=Stars_TexturedWireFrameEmissiveStrength}
    B=Multiply(A=@n2.RGB, B=@n6.Result)
   )
   CustomizedUV1=ComponentMask(None=PreSkinnedPosition()) {b=true, r=true}
   BaseColor=LinearInterpolate(A=@n7.RGB, B=@n6.Result, Alpha=@n2.RGB)
   Metallic=@n8.R
   Specular=@n8.R
   Roughness=@n8.R
   Normal=@n9.Result
  ) {attributeSetTypes [
    #1
    #4
    #5
    #6
    #7
    #8
    #3
   ]}
  n11 = Floor(None=ComponentMask(None=TextureCoordinate()) {r=true})
  n12 = VectorParameter() {group=Stars, parameterName=Stars_MSR_outer, defaultValue=<rgba 0 0 0.006 0>}
  n13 = ScalarParameter() {group=Stars, parameterName=Stars Inner Offset}
  n14 = StaticSwitchParameter(
   True=MaterialFunctionCall() {materialFunction=/Game/Characters/Player/Female/Medium/Bodies/F_MED_Stars/MF_stars_handsOffset}.Result
   False=Constant()
  ) {group=StarsHands, parameterName=Stars OffsetHands, sortPriority=0}
  n15 = Multiply(A=ScalarParameter() {group=StarsHands, parameterName=Stars Hands InnerOffset, sortPriority=1}, B=@n14)
  n16 = ScalarParameter() {group=Stars, parameterName=Stars Outer Offset}
  n17 = Multiply(A=ScalarParameter() {group=StarsHands, parameterName=Stars Hands OuterOffset, sortPriority=1}, B=@n14)
  n18 = MaterialFunctionCall(Offset Amount=LinearInterpolate(A=Add(A=@n13, B=@n15), B=Add(A=@n16, B=@n17), Alpha=@n11) {desc="A : inner, B: outer"}) {materialFunction=/Engine/Functions/Engine_MaterialFunctions02/WorldPositionOffset/CameraOffset}
  n19 = Multiply(A=Subtract(A=WorldPosition().XYZ, B=ObjectPositionWS())) {constB=-1}
  n20 = SetMaterialAttributes!(
   MaterialAttributes=@n1.MaterialAttributes
   BaseColor=LinearInterpolate(A=@n7.RGB, B=VectorParameter() {group=Stars, parameterName=Stars_BaseColor_outer, defaultValue=<rgba 0.002 0.002 0.002 0>}.RGB, Alpha=@n11)
   Metallic=LinearInterpolate(A=@n8.R, B=@n12.R, Alpha=@n11)
   Specular=LinearInterpolate(A=@n8.G, B=@n12.G, Alpha=@n11)
   Roughness=LinearInterpolate(A=@n8.B, B=@n12.B, Alpha=@n11)
   EmissiveColor=LinearInterpolate(
    A=VectorParameter() {group=Stars, parameterName=stars-emissive_lowRes}.RGB
    B=VectorParameter() {group=Stars, parameterName=stars_emissive_outer, defaultValue=<rgba 0.005 0.005 0.005 0>}.RGB
    Alpha=@n11
   )
   WorldPositionOffset=RayTracingQualitySwitch(Normal=@n18.Camera Offset, RayTraced=LinearInterpolate(A=@n18.Camera Offset, B=@n19, Alpha=@n11) {desc="A : inner, B: outer"})
   Normal=@n9.Result
   OpacityMask=RayTracingQualitySwitch(Normal=Constant() {r=1}, RayTraced=ScalarParameter() {defaultValue=1, group=Stars, parameterName=Stars_RaytracedOpacity})
  ) {attributeSetTypes [
    #5
    #6
    #7
    #8
    #1
    #2
    #3
    #9
   ]}
  FunctionOutput(None=StaticSwitchParameter(True=@n10, False=@n20) {group=Stars, parameterName=Stars_TexturedWireframe})
 ]}
