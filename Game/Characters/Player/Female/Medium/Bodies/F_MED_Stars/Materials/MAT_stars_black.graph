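# Masked, default-lit character material for the black variant of the female medium "Stars" player body, built for skeletal meshes with morph targets and mesh deformers.
# It uses Material Attributes, passing the shared PreFX character function's result through MF_stars_black, which supplies the variant-specific material to the shared PostFX function.
# PreFX's AO output also drives all three of PostFX's AO, Distance Blend and Fresnel inputs.
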Material /Game/Characters/Player/Female/Medium/Bodies/F_MED_Stars/Materials/MAT_stars_black
graph {
 properties {
  bUseMaterialAttributes=true
  bUsedWithMeshDeformer=true
  bUsedWithMorphTargets=true
  bUsedWithSkeletalMesh=true
  blendMode=BLEND_Masked
  opacityMaskClipValue=0.3
  pixelDepthOffsetMode=PDOM_Legacy
  usedShadingModels=Default Lit
 }
 nodes [
  n1 = MaterialFunctionCall() {materialFunction=&BaseFN_CharacterFunction_PreFX}
  MaterialFunctionCall(
   AO=@n1.AO
   Distance Blend=@n1.AO
   Fresnel=@n1.AO
   Material=MaterialFunctionCall!(In=@n1.Result) {materialFunction=&/MF_stars_black}.Result
  ) {materialFunction=&BaseFN_CharacterFunction_PostFX}
 ]
}
