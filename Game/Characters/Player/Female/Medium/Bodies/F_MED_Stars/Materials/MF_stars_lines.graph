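# Material function that styles the line ribbons on the female Stars player body with animated, hue-cycling color.
# A three-color gradient pans along pre-skinned height, passed through CustomizedUV1, feeding base color and Fresnel-weighted emissive.
# The shared spline-thicken function drives normals and vertex offset, with ribbon width growing with camera distance and optionally scaled by object radius.
# A camera offset, with a switchable hands adjustment, is added to that vertex offset.
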MaterialFunction /Game/Characters/Player/Female/Medium/Bodies/F_MED_Stars/Materials/MF_stars_lines
graph {nodes [
  n1 = Time()
  n2 = TextureCoordinate() {coordinateIndex=1}
  n3 = Panner(Coordinate=Divide(A=ComponentMask(None=@n2) {g=true}) {constB=140}) {bFractionalPart=true, speedX=0.1}
  n4 = MaterialFunctionCall(
   Hue Shift Percentage=Frac(None=Multiply(A=@n1) {constB=0.2})
   Texture=MaterialFunctionCall(
    A=Constant3Vector() {constant=<rgba 0.1015 0 1 0>}
    B=Constant3Vector() {constant=<rgba 0.855 0.0343 0.7913 0>}
    C=Constant3Vector() {constant=<rgba 0.1022 0 1 1>}
    Alpha=Frac(None=@n3)
   ) {materialFunction=&3ColorBlend}.Result
  ) {materialFunction=&HueShift}
  n5 = ScalarParameter() {parameterName=Metallic}
  n6 = Fresnel() {baseReflectFraction=0.1, exponent=2}
  n7 = WorldPosition()
  n8 = CameraPositionWS()
  n9 = ScalarParameter() {defaultValue=200, group=Stars, parameterName=Stars_DepthMin}
  n10 = Divide(
   A=Subtract(A=Distance(A=@n7.XYZ, B=@n8), B=@n9)
   B=Subtract(A=ScalarParameter() {defaultValue=601, group=Stars, parameterName=Stars_DepthMax}, B=@n9)
  )
  n11 = ScalarParameter() {defaultValue=1, group=Stars, parameterName=Stars_RibbonWidthMult, sortPriority=1}
  n12 = Divide(A=ObjectRadius(), B=ScalarParameter() {defaultValue=97.3942, group=Stars, parameterName=ActorRadius, sortPriority=0})
  n13 = Multiply(
   A=LinearInterpolate(
    A=ScalarParameter() {defaultValue=0.3, group=Stars, parameterName=Stars_RibbonMinWidth}
    B=ScalarParameter() {defaultValue=0.5, group=Stars, parameterName=Stars_RibbonMaxWidth}
    Alpha=Saturate(None=@n10)
   )
   B=LinearInterpolate(A=@n11, B=Multiply(A=@n11, B=@n12), Alpha=ScalarParameter() {group=Stars, parameterName=useActorRadius, sortPriority=1})
  )
  n14 = MaterialFunctionCall!(widthTip=@n13, Flatten=ScalarParameter() {defaultValue=0.5, group=Stars, parameterName=Stars_RibbonFlatten}, widthBase=@n13) {materialFunction=/Game/Characters/Player/Male/Medium/Bodies/M_MED_Neon/Materials/MF_splineThickenM}
  n15 = MaterialFunctionCall() {materialFunction=/Game/Characters/Player/Female/Medium/Bodies/F_MED_Stars/MF_stars_handsOffset}
  n16 = Constant()
  n17 = Add(
   A=ScalarParameter() {defaultValue=0.5, group=Stars, parameterName=Lines_Offset}
   B=Multiply(
    A=ScalarParameter() {group=StarsHands, parameterName=Lines Hands Offset, sortPriority=1}
    B=StaticSwitchParameter(True=@n15.Result, False=@n16) {group=StarsHands, parameterName=Stars OffsetHands, sortPriority=0}
   )
  )
  n18 = SetMaterialAttributes!(
   MaterialAttributes=GetMaterialAttributes(None=FunctionInput() {inputType=FunctionInput_MaterialAttributes}) {attributeGetTypes [#1, #2]}.MaterialAttributes
   BaseColor=Multiply(A=@n4.Result)
   Metallic=@n5
   Specular=@n5
   EmissiveColor=Multiply(A=Multiply(A=@n4.Result, B=@n6))
   Normal=@n14.normal
   WorldPositionOffset=Add(
    A=MaterialFunctionCall(Offset Amount=@n17) {materialFunction=/Engine/Functions/Engine_MaterialFunctions02/WorldPositionOffset/CameraOffset}.Camera Offset
    B=@n14.worldPositionOffset
   )
   CustomizedUV1=ComponentMask(None=PreSkinnedPosition()) {b=true, r=true}
  ) {attributeSetTypes [
    #3
    #4
    #5
    #1
    #6
    #2
    #7
   ]}
  FunctionOutput(None=@n18)
 ]}
