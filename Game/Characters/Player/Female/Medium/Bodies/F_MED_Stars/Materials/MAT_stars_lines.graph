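# Skeletal-mesh material for the lines variant of the female medium "Stars" player body, two-sided Default Lit with morph target and mesh deformer support.
# It uses material attributes, running the shared character PreFX function's result through MF_stars_lines and then into the shared PostFX function.
# The PreFX AO output also drives the AO, Distance Blend and Fresnel inputs of PostFX.
# Normals are set to non-tangent space, with two customized UVs.
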Material /Game/Characters/Player/Female/Medium/Bodies/F_MED_Stars/Materials/MAT_stars_lines
graph {
 properties {
  bTangentSpaceNormal=false
  bUseMaterialAttributes=true
  bUsedWithMeshDeformer=true
  bUsedWithMorphTargets=true
  bUsedWithSkeletalMesh=true
  numCustomizedUVs=2
  pixelDepthOffsetMode=PDOM_Legacy
  twoSided=true
  usedShadingModels=Default Lit
 }
 nodes [
  n1 = MaterialFunctionCall() {materialFunction=&BaseFN_CharacterFunction_PreFX}
  MaterialFunctionCall(
   AO=@n1.AO
   Distance Blend=@n1.AO
   Fresnel=@n1.AO
   Material=MaterialFunctionCall!(In=@n1.Result) {materialFunction=&/MF_stars_lines}.Result
  ) {materialFunction=&BaseFN_CharacterFunction_PostFX}
 ]
}
