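# Unlit additive material that draws twinkling star sprites on the F_MED_Stars female player body.
# World position offset expands quads along camera axes from their UVs, sized by twinkle with optional object-radius scaling, plus a camera offset.
# Vertex red seeds panned noise into a curve atlas for twinkle, which, with view facing, blends flare and filter textures into opacity, zeroed for ray tracing.
# Event audio amplitude boosts both twinkle and purple emissive.
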Material /Game/Characters/Player/Female/Medium/Bodies/F_MED_Stars/Materials/MAT_stars_stars
graph {
 properties {
  bCastRayTracedShadows=false
  bUseMaterialAttributes=true
  bUsedWithMeshDeformer=true
  bUsedWithMorphTargets=true
  bUsedWithSkeletalMesh=true
  blendMode=BLEND_Additive
  ditherOpacityMask=true
  opacityMaskClipValue=0.2
  pixelDepthOffsetMode=PDOM_Legacy
  shadingModel=MSM_Unlit
  usedShadingModels=Unlit
 }
 nodes [
  n1 = CollectionParameter() {collection=&FortniteMaterialParameters, parameterName=EventAudio_Amplitude}
  n2 = MaterialFunctionCall!(Base=Multiply(A=VectorParameter() {defaultValue=<rgba 0.9199 0.5354 1 0>}.RGB) {constB=3}, Add=Power(Base=@n1) {constExponent=1}) {materialFunction=/Engine/Functions/Engine_MaterialFunctions02/Math/MultiplyAdd}
  n3 = TextureCoordinate()
  n4 = TextureSample(UVs=@n3) {texture=/Game/Effects/Fort_Effects/Textures/Flares/T_Flare_Spikeball}
  n5 = TextureSample(UVs=@n3) {texture=/Game/Characters/Player/Female/Medium/Bodies/F_MED_Stars/Textures/starFilter}
  n6 = DotProduct(A=CameraVectorWS(), B=PixelNormalWS())
  n7 = Power(Base=@n6) {constExponent=3}
  n8 = Panner(Coordinate=VertexColor().R, Speed=ScalarParameter() {defaultValue=0.06, parameterName=TwinkleSpeed})
  n9 = CurveAtlasRowParameter(CurveTime=Power(Base=TextureSample(UVs=@n8) {samplerType=SAMPLERTYPE_LinearColor, texture=&HighContrastNoise}.R)) {atlas=&AthenaCurveAtlas01, curve=/Game/Athena/LinearColorCurves/Curve_Galileo_Hologram, defaultValue=171}
  n10 = CollectionParameter() {collection=&FortniteMaterialParameters, parameterName=EventAudio_Amplitude}
  n11 = MaterialFunctionCall(Base=@n9.B, Add=Power(Base=@n10) {constExponent=4}) {materialFunction=&MultiplyAdd}
  n12 = Multiply(A=Multiply(A=Saturate(None=@n7), B=Saturate(None=@n11.Result))) {constB=2}
  n13 = RayTracingQualitySwitch(Normal=Multiply(A=LinearInterpolate(A=@n4.R, B=@n5.R, Alpha=@n12)), RayTraced=Constant())
  n14 = MaterialFunctionCall(Float2=Subtract(A=TextureCoordinate()) {constB=0.5}) {materialFunction=&BreakOutFloat2Components}
  n15 = Multiply(A=@n14.R, B=Transform(None=Constant3Vector() {constant=<rgba 1 0 0 0>}) {transformSourceType=TRANSFORMSOURCE_Camera})
  n16 = ScalarParameter() {defaultValue=10, group=Stars, parameterName=Stars_starSizeMult}
  n17 = ObjectRadius()
  n18 = ScalarParameter() {defaultValue=97.3942, group=Stars, parameterName=ActorRadius, sortPriority=0}
  n19 = LinearInterpolate(A=@n16, B=Multiply(A=@n16, B=Divide(A=@n17, B=@n18)), Alpha=ScalarParameter() {group=Stars, parameterName=useActorRadius, sortPriority=1})
  n20 = MaterialFunctionCall(Float2=Add(A=ScalarParameter() {group=Stars, parameterName=Stars_starMinSize}, B=Multiply(A=@n11.Result, B=@n19) {constB=10})) {materialFunction=&BreakOutFloat2Components}
  n21 = Multiply(A=@n14.G, B=Transform(None=Constant3Vector() {constant=<rgba 0 -1 0 0>}) {transformSourceType=TRANSFORMSOURCE_Camera})
  n22 = Add(
   A=Add(A=Multiply(A=@n15, B=@n20.R), B=Multiply(A=@n21, B=@n20.G))
   B=MaterialFunctionCall(Offset Amount=ScalarParameter() {defaultValue=3, group=Stars, parameterName="Stars Offset Amount (S)"}) {materialFunction=/Engine/Functions/Engine_MaterialFunctions02/WorldPositionOffset/CameraOffset}.Clamped Camera Offset
  )
  MaterialFunctionCall(In=SetMaterialAttributes!(
    MaterialAttributes
    Emissive Color=NamedRerouteDeclaration(None=@n2.Result) {name=emissive, nodeColor=<rgba 0.6406 0 1 1>}
    Opacity=NamedRerouteDeclaration(None=@n13) {name=opacity, nodeColor=<rgba 0 0.3672 1 1>}
    World Position Offset=NamedRerouteDeclaration(None=@n22) {name=wpo, nodeColor=<rgba 0 1 0.5078 1>}
   ) {attributeSetTypes [#1, #2, #3]}) {materialFunction=&MF_TechArtOverrides}
  OneMinus(None=@n6)
 ]
}
